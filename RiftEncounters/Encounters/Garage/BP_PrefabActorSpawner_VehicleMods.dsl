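; Prefab actor spawner for the Garage rift encounter that fits mods onto the vehicle it spawns.
; On BeginPlay it registers a callback for when the actor spawns.
; The callback waits a short fixed delay, casts the actor to FortAthenaVehicle, then applies the tire and cow catcher mods from tag variables.
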(event EventBeginPlay
  (CallFunction|CallwhenActorSpawned (delegate Custom|ApplyVehicleMods)))

(event Collision|EventActorBeginOverlap
  (params (OtherActor :type Actor%20Object%20Reference)))

(event EventTick
  (params (DeltaSeconds :type Float%20%28single-precision%29)))

(event EventEndPlay
  (params (EndPlayReason :type EEndPlayReason%20Enum :default "Destroyed")))

(event Custom|ApplyVehicleMods
  (params (SpawnedActor :type Actor%20Object%20Reference))
  (Utilities|FlowControl|Delay 0.7)
  (bind _asfort_athena_vehicle (Utilities|Casting|CastToFortAthenaVehicle SpawnedActor))
  (Vehicle|Mod|ApplyVehicleMod _asfort_athena_vehicle (Variables|Default|GetTire%20Mod%20Tag) 0)
  (Vehicle|Mod|ApplyVehicleMod _asfort_athena_vehicle (Variables|Default|GetCow%20Catcher%20Mod%20Tag) 0))

(fn ConstructionScript)
